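# Desert Point dune material that blends sand meshes into the Apollo landscape through color and height runtime virtual textures.
# By default, vertex-color height lerps mix sand tones and normals, while mobile/forward tints shared biome sand with radial sphere masks.
# UseTextureSet swaps in textured shading with optional detail and grass, plus hit-effect WPO.
# Tangent normals become world-space, then height above terrain fades RVT color and normals, except on mobile/forward.
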Material /Game/Environments/Apollo/Sets/DesertPoint/Dunes/Materials/M_RVT_Dunes
graph {
 properties {bTangentSpaceNormal=false, bUseMaterialAttributes=true, pixelDepthOffsetMode=PDOM_Legacy, usedShadingModels=Default Lit}
 nodes [
  n1 = TextureCoordinate()
  n2 = Constant() {r=5}
  n3 = MaterialFunctionCall(
   Transition Phase=VertexColor().R
   Height Texture=TextureSample(UVs=Multiply(A=@n1, B=@n2)) {samplerType=SAMPLERTYPE_Grayscale, texture=&LowResBlurredNoise}.R
  ) {materialFunction=&HeightLerp}
  LinearInterpolate(B=VectorParameter() {group=Mobile, parameterName=ToneMobile, defaultValue=<rgba 0 0.1547 2 0>}.RGB, Alpha=@n3.Lerp Alpha No Contrast)
  n4 = WorldPosition()
  n5 = TextureSampleParameter2D(UVs=Divide(A=ComponentMask(None=@n4.XYZ) {g=true, r=true}, B=ScalarParameter() {defaultValue=512, group=Grass, parameterName=ScaleGrass})) {group=Grass, parameterName=Grass, texture=/Game/Packages/DS_Fortnite_Terrain_NoLOD/Textures/T_Grasslands_D}
  n6 = LinearInterpolate(
   A=ScalarParameter() {defaultValue=0.6, group=Grass, parameterName=Grass R2}
   B=ScalarParameter() {defaultValue=0.6, group=Grass, parameterName=Grass R}
   Alpha=@n5.G
  )
  n7 = VectorParameter() {parameterName=Tone, defaultValue=<rgba 0.3519 0.3435 0.5234 0>}
  n8 = TextureSampleParameter2D() {
   group=Textures
   parameterName=Diffuse
   samplerSource=SSM_Wrap_WorldGroupSettings
   sortPriority=1
   texture=/Game/Global/Textures/Default/Blanks/T_MidgreysRGB
  }
  n9 = TextureSampleParameter2D() {
   group=Textures
   parameterName=SpecularMasks
   samplerSource=SSM_Wrap_WorldGroupSettings
   samplerType=SAMPLERTYPE_Masks
   sortPriority=-4
   texture=&T_Fortnite_Default_S
  }
  n10 = TextureSampleParameter2D() {
   group=Textures
   parameterName=Normals
   samplerSource=SSM_Wrap_WorldGroupSettings
   samplerType=SAMPLERTYPE_Normal
   texture=/Game/Environments/Apollo/Sets/Terrain/Textures/T_Apollo_Sand_N_01
  }
  n11 = SetMaterialAttributes!(
   MaterialAttributes
   Base Color=Multiply(A=@n7.RGB, B=@n8.RGB)
   Specular=Multiply(A=ScalarParameter() {defaultValue=1, group=Textures, parameterName=Specular}, B=@n9.R)
   Metallic=@n9.G
   Roughness=Multiply(A=@n9.R, B=ScalarParameter() {defaultValue=1, group=Textures, parameterName=Roughness})
   Normal=@n10.RGB
   Opacity Mask=@n8.A
  ) {attributeSetTypes [
    #1
    #2
    #3
    #4
    #5
    #6
   ]}
  n12 = GetMaterialAttributes(None=@n11) {attributeGetTypes [#1, #2, #4]}
  n13 = ObjectRadius()
  n14 = ScalarParameter() {defaultValue=300, group=Detail Texture, parameterName=DetailTile, sortPriority=3}
  n15 = TextureSampleParameter2D(UVs=Multiply(A=TextureCoordinate(), B=Divide(A=@n13, B=@n14))) {group=Detail Texture, parameterName=Details, texture=/Game/Environments/Apollo/Rocks/Natural_Rocks/Textures/T_RockDetails_02D}
  n16 = LinearInterpolate(
   A=ScalarParameter() {defaultValue=-0.773, group=Detail Texture, parameterName=Detail R Min, sortPriority=5}
   B=ScalarParameter() {defaultValue=1.8112, group=Detail Texture, parameterName=Detail R Max, sortPriority=6}
   Alpha=@n15.R
  )
  n17 = StaticSwitchParameter(
   True=SetMaterialAttributes!(
    MaterialAttributes=@n12.MaterialAttributes
    Base Color=MaterialFunctionCall(Base=@n12.Base Color, Blend=@n15.RGB) {materialFunction=&Blend_Overlay}.Result
    Specular=Multiply(A=@n12.Specular, B=@n15.R)
    Roughness=Multiply(A=@n12.Roughness, B=@n16)
   ) {attributeSetTypes [#1, #2, #4]}
   False=@n11
  ) {defaultValue=true, desc=Uses regular textures setup, group=Switches, parameterName=UseDetailTextures}
  n18 = ShadingPathSwitch(Default=@n17, Deferred=@n17, Forward=@n11, Mobile=@n11)
  n19 = StaticSwitchParameter(
   True=BlendMaterialAttributes(
    A=SetMaterialAttributes(Base Color=@n5.RGB, Specular=@n5.RGB, Roughness=@n6) {attributeSetTypes [#1, #2, #4]}
    B=@n18
    Alpha=VertexColor().R
   ) {vertexAttributeBlendType=UseA}
   False=@n18
  ) {desc=uses black Vertex color, group=Switches, parameterName=Grass True}
  n20 = MaterialFunctionCall!(
   WSDeformationCenter=VectorParameter() {bUseCustomPrimitiveData=true, group=z_code, parameterName=WSDeformationCenter, primitiveDataIndex=20}.RGB
   Code_DestructionAmount=ScalarParameter() {group=z_code, parameterName=Code_DestructionAmount}
   Code_UseUvForDamage
   Code_UseHitEffects=StaticBoolParameter() {defaultValue=true, group=z_code, parameterName=Code_UseHitEffects}
   Code_Emissive
   WSDamageRadius=ScalarParameter() {bUseCustomPrimitiveData=true, defaultValue=1, group=z_code, parameterName=WSDamageRadius, primitiveDataIndex=26}
   CrackDamageShader
   Diffuse
   ScalarModDamageCracks
   ScalarOffsetforDamageCracks
   ExposeLerpFunctionForInvertingTheHitDistanceCalc
   InvertDamageFalloff
   CrackWorldPosition
   WorldPosition
   DistanceCullFadeAnimationOffset=ScalarParameter() {bUseCustomPrimitiveData=true, group=z_code, parameterName=DistanceCullFadeAnimationOffset, primitiveDataIndex=12}
  ) {materialFunction=/Game/Packages/jl_tests/Fortnite_HitEffects}
  n21 = MaterialFunctionCall!(CustomSeaLevel, BypassBiomeColors, UseOceanLevel) {materialFunction=/Game/Athena/Apollo/Environments/Landscape/Materials/Biome/CommonMaterials/Mobile/MF_Mobile_Common_Sand}
  n22 = GetMaterialAttributes(None=@n21.Material) {attributeGetTypes [#1]}
  n23 = WorldPosition()
  n24 = VectorParameter() {parameterName=Zero, defaultValue=<rgba 7400 5800 0 0>}
  n25 = ScalarParameter() {defaultValue=12077.8486, group=Mobile, parameterName=MobileRadius}
  n26 = ScalarParameter() {defaultValue=0.5, group=Mobile, parameterName=MobileHardness}
  n27 = SphereMask(
   A=@n23.XYZ
   B=@n24.RGB
   Radius=ScalarParameter() {defaultValue=25555, group=Mobile, parameterName=MobileRadiusFar}
   Hardness=ScalarParameter() {defaultValue=0.25, group=Mobile, parameterName=MobileHardnessFar}
  )
  n28 = LinearInterpolate(
   A=VectorParameter() {group=Mobile, parameterName=ColorFarMobile, defaultValue=<rgba 0.8113 0.4232 3 0>}.RGB
   B=LinearInterpolate(
    A=VectorParameter() {group=Mobile, parameterName=ColorCenterMobile, defaultValue=<rgba 0.6654 0.245 0.7969 0>}.RGB
    B=VectorParameter() {group=Mobile, parameterName=ColorEyeMobile, defaultValue=<rgba 0.0374 0.0179 0.0547 0>}.RGB
    Alpha=SphereMask(A=@n23.XYZ, B=@n24.RGB, Radius=@n25, Hardness=@n26)
   )
   Alpha=@n27
  )
  n29 = BlendMaterialAttributes(
   A=@n21.Material
   B=SetMaterialAttributes!(
    MaterialAttributes=@n22.MaterialAttributes
    Base Color=Multiply(A=@n22.Base Color, B=@n28)
    Normal=Constant3Vector() {constant=<rgba 0 0 12 1>}
   ) {attributeSetTypes [#1, #5]}
   Alpha=@n27
  )
  n30 = VectorParameter() {parameterName=Tone2, defaultValue=<rgba 0.7057 0.4528 0.294 0>}
  n31 = ScalarParameter() {defaultValue=0.25, parameterName=AddToDif}
  n32 = WorldPosition()
  n33 = TextureSample(UVs=Divide(A=ComponentMask(None=@n32.XYZ) {g=true, r=true}, B=ScalarParameter() {defaultValue=-255, parameterName=ScaleSandSpecl})) {samplerType=SAMPLERTYPE_Masks, texture=/Game/Environments/Apollo/Sets/DesertPoint/Impact/NewSand_M}
  n34 = TextureSampleParameter2D(UVs=Multiply(A=TextureCoordinate(), B=ScalarParameter() {defaultValue=12, group=Textures, parameterName=ScaleNormals})) {
   group=Textures
   parameterName=Normals
   samplerType=SAMPLERTYPE_Normal
   texture=&T_Apollo_Sand_N_01
  }
  n35 = ScalarParameter() {defaultValue=1, desc=It will be scaled by object size if you scale mesh, group=Textures, parameterName=Scale normals 2}
  n36 = Divide(A=ObjectRadius()) {constB=1024}
  n37 = TextureSampleParameter2D(UVs=Multiply(A=TextureCoordinate(), B=Multiply(A=@n35, B=@n36))) {group=Textures, parameterName=Normals2, samplerType=SAMPLERTYPE_Normal, texture=/Game/Environments/Sets/Arid/Textures/T_desert_sand_wavy_N}
  n38 = Multiply(A=@n37.RGB, B=VectorParameter() {parameterName=Nomals2Value, defaultValue=<rgba 0.25 0.25 1 1>}.RGB)
  n39 = VertexNormalWS()
  n40 = SetMaterialAttributes!(
   MaterialAttributes
   Base Color=Add(A=LinearInterpolate(A=@n30.RGB, B=@n7.RGB, Alpha=@n3.Lerp Alpha No Contrast), B=Multiply(A=@n31, B=@n33.G))
   Normal=LinearInterpolate(
    A=Constant3Vector() {constant=<rgba 0 0 1 0>}
    B=LinearInterpolate(A=@n34.RGB, B=@n38, Alpha=@n3.Lerp Alpha No Contrast)
    Alpha=ComponentMask(None=@n39) {b=true}
   )
   Roughness=LinearInterpolate(A=ScalarParameter() {defaultValue=-1, parameterName=R2 Center}, B=ScalarParameter() {defaultValue=1, parameterName=R1 Center}, Alpha=@n33.B)
   Specular=@n33.R
  ) {attributeSetTypes [
    #1
    #5
    #4
    #2
   ]}
  n41 = MaterialFunctionCall!(Material=StaticSwitchParameter(
    True=SetMaterialAttributes!(MaterialAttributes=@n19, World Position Offset=@n20.Hit And Shrink WPO) {attributeSetTypes [#7]}
    False=ShadingPathSwitch(Default=@n40, Deferred=@n40, Forward=@n29, Mobile=@n29)
   ) {desc=Uses regular textures setup, group=Switches, parameterName=UseTextureSet}) {materialFunction=/Game/Athena/Apollo/Maps/Landscape/VT/MaterialFunctions/TransformTangentToWorld}
  n42 = RuntimeVirtualTextureSampleParameter() {materialType=BaseColor_Normal_Specular_YCoCg, parameterName=RVT, virtualTexture=/Game/Athena/Apollo/Maps/Landscape/VT/RVT_Apollo_Landscape}
  n43 = RuntimeVirtualTextureSampleParameter() {materialType=WorldHeight, parameterName=RVTHeight, virtualTexture=/Game/Athena/Apollo/Maps/Landscape/VT/RVT_Apollo_Landscape_Height}
  n44 = Subtract(A=ComponentMask(None=WorldPosition().XYZ) {b=true}, B=@n43.WorldHeight)
  n45 = ScalarParameter() {defaultValue=750, group=RVT, parameterName=Blend RVT Normals}
  n46 = BlendMaterialAttributes(
   A=SetMaterialAttributes!(MaterialAttributes=@n41.Result, Normal=@n42.Normal) {attributeSetTypes [#5]}
   B=@n41.Result
   Alpha=Saturate(None=Divide(A=@n44, B=@n45))
  ) {vertexAttributeBlendType=UseA}
  n47 = SetMaterialAttributes!(MaterialAttributes=@n46, Base Color=@n42.BaseColor, Specular=@n42.BaseColor, Roughness=@n42.BaseColor) {attributeSetTypes [#1, #2, #4]}
  n48 = ScalarParameter() {defaultValue=2000, group=RVT, parameterName=Blend RVT}
  n49 = LinearInterpolate(A=ScalarParameter() {defaultValue=2000, group=RVT, parameterName=Blend RVT Top}, B=@n48, Alpha=VertexColor().R)
  n50 = Saturate(None=StaticSwitchParameter(True=Divide(A=@n44, B=@n48), False=Divide(A=@n44, B=@n49)) {desc=Uses regular textures setup, group=Switches, parameterName=UseTextureSet})
  n51 = MaterialFunctionCall(
   NoRVT=@n41.Result
   RVT=MaterialProxyReplace(Realtime=BlendMaterialAttributes(A=@n47, B=@n46, Alpha=@n50) {vertexAttributeBlendType=UseA}, MaterialProxy=@n41.Result)
  ) {materialFunction=&FN_RVT_Switch_Material}
  ShadingPathSwitch(Default=@n51.Result, Deferred=@n51.Result, Forward=@n41.Result, Mobile=@n41.Result)
 ]
}
